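# Sound cue for the landing sound of the WinterScoped suppressed sniper rifle, using weapon-drop attenuation and the loot-collect sound class.
# It plays a single land wave through a modulator node, with raised pitch and reduced volume set on the cue.
# A concurrency override prevents new instances from starting once the limit is reached.
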SoundCue /Game/Sounds/Fort_Weapons/SniperRifles/WinterScoped/SC_SuppressedSniper_Landed
root=Modulator_0
nodes [
 {id=Modulator_0, children [WavePlayer_0]}
 {id=WavePlayer_0, properties {soundWaveAssetPtr=&/SniperSuppressed_Land}}
]
properties {
 attenuationSettings=&Fort_Weapon_Drop_Att
 bOverrideConcurrency=true
 maxDistance=3000
 pitchMultiplier=1.1
 soundClassObject=&Fort_Loot_Collect
 volumeMultiplier=0.4308
 concurrencyOverrides {resolutionRule=PreventNew}
}
