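# Sound cue for firing the winter scoped sniper rifle, layering close and medium-distance gunfire, apparently as heard from other players.
# The close branch mixes a random shot with a quiet, delayed random eject impact sound, using close projectile attenuation and a third-person fire sound class.
# The medium branch plays a pitched-up, boosted random shot through mono medium attenuation and the distant fire class.
# Local and global gunfire concurrency sets are assigned.
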SoundCue /Game/Sounds/Fort_Weapons/SniperRifles/WinterScoped/Winter_Sniper_Fire_Cue
root=Mixer_1
nodes [
 {
  id=Attenuation_118
  children [Modulator_0]
  properties {attenuationSettings=&Fort_Weapon_Projectile_Close_Athena_Att}
 }
 {
  id=Attenuation_119
  children [Modulator_2]
  properties {attenuationSettings=&Fort_Weapon_Projectile_Medium_Mono_Athena_Att}
 }
 {id=Delay_0, children [Random_8], properties {delayMax=0.7, delayMin=0.6}}
 {id=Mixer_1, children [SoundClass_77, SoundClass_76], properties {inputVolume [0.7,0.5]}}
 {id=Mixer_2, children [Random_6,Modulator_1], properties {inputVolume [1,1]}}
 {id=Modulator_0, children [Mixer_2], properties {volumeMax=1, volumeMin=0.9}}
 {id=Modulator_1, children [Delay_0], properties {pitchMax=1, pitchMin=0.9, volumeMax=0.2, volumeMin=0.1}}
 {id=Modulator_2, children [Random_0], properties {pitchMax=1.2, pitchMin=1.1, volumeMax=2.05, volumeMin=1.95}}
 {id=Random_0, children [WavePlayer_0,WavePlayer_1,WavePlayer_2], properties {bShouldExcludeFromBranchCulling=true, weights [1,1,1]}}
 {id=Random_6, children [WavePlayer_13, WavePlayer_14, WavePlayer_15], properties {bShouldExcludeFromBranchCulling=true, weights [1,1,1]}}
 {id=Random_8, children [WavePlayer_22, WavePlayer_23, WavePlayer_24], properties {bShouldExcludeFromBranchCulling=true, weights [1,1,1]}}
 {
  id=SoundClass_76
  children [Attenuation_119]
  properties {soundClassOverride=&Fort_WeaponRanged_Fire_Distant}
 }
 {
  id=SoundClass_77
  children [Attenuation_118]
  properties {soundClassOverride=&Fort_WeaponRanged_Fire_3P_PSM}
 }
 {id=WavePlayer_0, properties {soundWaveAssetPtr=&/Winter_Sniper_Fire_MediumNew_02}}
 {id=WavePlayer_1, properties {soundWaveAssetPtr=&/Winter_Sniper_Fire_MediumNew_01}}
 {id=WavePlayer_13, properties {soundWaveAssetPtr=&/Winter_Sniper_FireClose_01}}
 {id=WavePlayer_14, properties {soundWaveAssetPtr=&/Winter_Sniper_FireClose_02}}
 {id=WavePlayer_15, properties {soundWaveAssetPtr=&/Winter_Sniper_FireClose_03}}
 {id=WavePlayer_2, properties {soundWaveAssetPtr=&/Winter_Sniper_Fire_MediumNew_03}}
 {id=WavePlayer_22, properties {soundWaveAssetPtr=&/Winter_Sniper_ReloadEjectImpact_01}}
 {id=WavePlayer_23, properties {soundWaveAssetPtr=&/Winter_Sniper_ReloadEjectImpact_02}}
 {id=WavePlayer_24, properties {soundWaveAssetPtr=&/Winter_Sniper_ReloadEjectImpact_03}}
]
properties {
 bDebug=true
 maxDistance=10579.2871
 soundClassObject=&Fort_WeaponRanged_Fire_3P_PSM
 volumeMultiplier=1
 concurrencyOverrides {bLimitToOwner=true, maxCount=3, resolutionRule=StopOldest}
 concurrencySet [&Weapon_Gunfire_Local_Conc, &Weapon_Gunfire_Global_Conc]
}
