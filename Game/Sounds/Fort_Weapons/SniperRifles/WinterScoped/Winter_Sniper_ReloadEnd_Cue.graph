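# Sound cue for the end of the reload on the Winter scoped sniper rifle.
# A single wave player plays Winter_Sniper_ReloadEnd_01 at reduced volume, with no randomization or mixing.
# It uses the weapon foley sound class, reload attenuation and a concurrency override, and has the debug flag left on.
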SoundCue /Game/Sounds/Fort_Weapons/SniperRifles/WinterScoped/Winter_Sniper_ReloadEnd_Cue
root=WavePlayer_2
nodes [{id=WavePlayer_2, properties {soundWaveAssetPtr=&/Winter_Sniper_ReloadEnd_01}}]
properties {
 attenuationSettings=&Fort_Weapon_Reload_Att
 bDebug=true
 bOverrideConcurrency=true
 maxDistance=2250
 soundClassObject=&Fort_WeaponFoley
 volumeMultiplier=0.7
}
