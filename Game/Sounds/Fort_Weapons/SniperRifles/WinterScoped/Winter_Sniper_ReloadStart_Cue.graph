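# Sound cue for the reload-start sound of the winter scoped sniper rifle.
# A mixer layers a shared bolt-action reload wave, played immediately, with a winter sniper reload-start wave that fires after a short fixed delay.
# It uses weapon reload attenuation, the weapon foley sound class, a reduced volume and overridden concurrency, with the debug flag enabled.
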SoundCue /Game/Sounds/Fort_Weapons/SniperRifles/WinterScoped/Winter_Sniper_ReloadStart_Cue
root=Mixer_0
nodes [
 {id=Delay_0, children [WavePlayer_0], properties {delayMax=0.3, delayMin=0.3}}
 {id=Mixer_0, children [Delay_0,WavePlayer_1], properties {inputVolume [1,1]}}
 {id=WavePlayer_0, properties {soundWaveAssetPtr=&/Winter_Sniper_ReloadStart_02}}
 {id=WavePlayer_1, properties {soundWaveAssetPtr=/Game/Sounds/Fort_Weapons/AutoRifle/Reload_Elements/rifle_sniper_reload_bolt_action}}
]
properties {
 attenuationSettings=&Fort_Weapon_Reload_Att
 bDebug=true
 bOverrideConcurrency=true
 maxDistance=2250
 soundClassObject=&Fort_WeaponFoley
 volumeMultiplier=0.451
}
